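# Masked, two-sided lit metal material for the RoyalAngst skydiving contrail, drawing chain textures on Niagara ribbons, sprites and meshes.
# Dynamic parameters drive SubUV frame, tiling and U offset for normal, SRM and RGBA samples; only U scale and frame count apply to both axes.
# A V-based cylindrical bend rounds the flattened normal, while a view-angle curve-atlas LUT tints the desaturated color.
# A normal-derived mask, shifted by particle Erosion, subtracts from alpha.
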Material /BRCosmetics/Athena/Effects/Particles/Skydiving/NS_Contrail_RoyalAngst/Materials/M_Contrail_RoyalAngst_Metal
graph {
 properties {
  bUsedWithNiagaraMeshParticles=true
  bUsedWithNiagaraRibbons=true
  bUsedWithNiagaraSprites=true
  blendMode=BLEND_Masked
  pixelDepthOffsetMode=PDOM_Legacy
  twoSided=true
  usedShadingModels=Default Lit
 }
 nodes [
  n1 = TextureObjectParameter() {
   parameterName=Tex Normal
   samplerType=SAMPLERTYPE_Normal
   texture=/BRCosmetics/Athena/Effects/Particles/Skydiving/NS_Contrail_RoyalAngst/Textures/T_RoyalAngst_Chains_N
  }
  n2 = DynamicParameter() {defaultValue=<rgba 1 5 1 1>, paramNames [U Frame Number, V Frame Number, U Scale, V Scale]}
  n3 = TextureCoordinate()
  n4 = ComponentMask(None=@n3) {r=true}
  n5 = DynamicParameter() {parameterIndex=1, defaultValue=<rgba 0 0 0 0>, paramNames [Frame, U Offset, Erosion, 4]}
  n6 = Multiply(A=AppendVector(A=@n2.U Scale, B=@n2.U Scale), B=AppendVector(A=Add(A=@n4, B=@n5.U Offset), B=ComponentMask(None=@n3) {g=true}))
  n7 = AppendVector(A=@n2.U Frame Number, B=@n2.U Frame Number)
  n8 = MaterialFunctionCall(
   Normal=MaterialFunctionCall!(Texture=@n1, UVs=@n6, SubImages=@n7, Frame=@n5.Frame) {materialFunction=/Engine/Functions/Engine_MaterialFunctions03/Texturing/SubUV_Function_MipDerivative}.RGB
   Flatness=Constant() {r=0.4}
  ) {materialFunction=&FlattenNormal}
  n9 = Constant()
  n10 = ConstantBiasScale(None=ComponentMask(None=TextureCoordinate()) {g=true}) {bias=-0.5, scale=2}
  n11 = DotProduct(A=@n10, B=@n10)
  n12 = OneMinus(None=Power(Base=OneMinus(None=@n11), Exp=ScalarParameter() {parameterName=Cylindrical Normal Intensity}))
  n13 = Sign(None=@n10)
  Add(
   A=@n8.Result
   B=MaterialFunctionCall(R=@n9, G=Multiply(A=@n12, B=@n13), B=@n9) {materialFunction=&AppendMany}.RGB
  )
  n14 = TextureObjectParameter() {
   parameterName=Tex SRM
   samplerType=SAMPLERTYPE_Masks
   texture=/BRCosmetics/Athena/Effects/Particles/Skydiving/NS_Contrail_RoyalAngst/Textures/T_RoyalAngst_Chains_SRM
  }
  n15 = MaterialFunctionCall(Texture=@n14, UVs=@n6, SubImages=@n7, Frame=@n5.Frame) {materialFunction=&SubUV_Function_MipDerivative}
  Clamp(None=ComponentMask(None=@n15.RGB) {g=true}) {minDefault=0.4}
  ComponentMask(None=@n15.RGB) {b=true}
  ComponentMask(None=@n15.RGB) {r=true}
  n16 = CameraVectorWS()
  n17 = PixelNormalWS()
  n18 = CurveAtlasRowParameter(CurveTime=Clamp(None=DotProduct(A=@n16, B=@n17))) {
   atlas=/Game/Global/Textures/Default/Curve/FN_Character_Atlas
   curve=/Game/Characters/Player/Female/Medium/Bodies/F_MED_Ms_Alpine/Textures/AluminiumreflectanceCurve
   group=Metal LUT
   parameterName=Metal LUT Curve
  }
  n19 = ChannelMaskParameter(None=@n18) {group=Metal LUT, parameterName=LUTChannel}
  n20 = TextureObjectParameter() {parameterName=Texture RGBA, texture=/BRCosmetics/Athena/Effects/Particles/Skydiving/NS_Contrail_RoyalAngst/Textures/T_RoyalAngst_Chains_RGBA}
  n21 = MaterialFunctionCall(Texture=@n20, UVs=@n6, SubImages=@n7, Frame=@n5.Frame) {materialFunction=&SubUV_Function_MipDerivative}
  n22 = Desaturation(None=@n21.RGB, Fraction=ScalarParameter() {parameterName=Desaturation})
  n23 = ScalarParameter() {defaultValue=1, group=Metal LUT, parameterName=MetalLutIntensity}
  Multiply(A=LinearInterpolate(B=Multiply(A=@n19.R, B=@n22), Alpha=Saturate(None=@n23))) {constB=0.6}
  ScalarParameter() {defaultValue=0.8, parameterName=Erosion, sliderMax=1}
  n24 = ComponentMask(None=@n8.Result) {r=true}
  n25 = ComponentMask(None=@n8.Result) {g=true}
  n26 = Multiply(A=OneMinus(None=Saturate(None=@n24)), B=OneMinus(None=Saturate(None=@n25)))
  n27 = LinearInterpolate(Alpha=@n5.Erosion) {constA=-1}
  Subtract(A=@n21.Alpha, B=Saturate(None=Add(A=@n26, B=@n27) {constB=-1}))
 ]
}
