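# Niagara skydiving contrail for the RoyalAngst cosmetic, hanging animated ribbon chains and glints from the player's hands and feet.
# Chain emitters burst persistent per-group ribbons at skeletal mesh bones, trailing them along velocity with sine sway.
# CrossHands_Chains joins both hands with Hermite splines, adding glide-lean response, pendulum constraints and curl noise.
# Glint emitters read chain particles to place twinkling sprites, colors scale by time of day, and a Lights_Flashing emitter follows.
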NiagaraSystem /BRCosmetics/Athena/Effects/Particles/Skydiving/NS_Contrail_RoyalAngst/NS_Contrail_RoyalAngst
dependencies {
 usedDataInterfaces [
  &NiagaraDataInterfaceArrayFloat
  &NiagaraDataInterfaceCurve
  &NiagaraDataInterfaceArrayFloat3
  &NiagaraDataInterfaceParticleRead
  &NiagaraDataInterfaceArrayInt32
  &NiagaraDataInterfaceArrayColor
 ]
 usedDynamicInputs [/Niagara/DynamicInputs/Rotation/InvertQuaternion]
 usedModules [
  &self:SystemSpawnScript.SystemScriptSource.SystemScriptGraph.NiagaraNodeAssignment_2.SetVariables_#1
  &self:SystemSpawnScript.SystemScriptSource.SystemScriptGraph.NiagaraNodeAssignment_3.SetVariables_#2
  &SystemState
  /Game/Effects/Niagara/Modules/Skydiving/FN_DivingAttributes/v4/FN_DivingAttributes_v4
  &self:SystemSpawnScript.SystemScriptSource.SystemScriptGraph.NiagaraNodeAssignment_1.SetVariables_#3
  &self:SystemSpawnScript.SystemScriptSource.SystemScriptGraph.NiagaraNodeAssignment_0.SetVariables_#4
 ]
}
emitters {
 Chains_Feet {
  emitterSpawnScript [&FN_PreserveSkydiveUserVariables() {module=PreserveUserVariables}]
  emitterUpdateScript [
   &EmitterState(
    Inactive Response=::NewEnumerator1
    Life Cycle Mode=::NewEnumerator1
    Loop Behavior=::NewEnumerator1
    Loop Duration Mode=::NewEnumerator1
    UseLoopDelay=false
   )
   /SetVariables(Emitter.Number of Chains=8, Emitter.Particle per Chain=5) {isSetParameters=true, module=SetVariables_#5}
   &SpawnBurst_Instantaneous(
    Spawn Count=&Multiply_Int(A=$Emitter.Number of Chains, B=$Emitter.Particle per Chain)
    Spawn Probability=HlslExpression() {hlslExpression="Emitter.LoopCount == 0"}
    Use Spawn Probability=false
   ) {module=SpawnBurst_Instantaneous001}
  ]
  particleSpawnScript [
   &FN_SetSpawnGroup(
    Inclusivity=true
    Normalize Spawn Group=true
    Particles Per Group:NiagaraInt32=$Emitter.Particle per Chain
    Pattern=::NewEnumerator0
    Set Ribbon ID and Ribbon Link Order=true
   )
   &InitializeParticle(
    Ribbon Width=&SelectFloatFromArray(
     Direct Array Index=&SpawnGroup()
     Float Selection Array=DataInterface() {propertyValues {FloatData [2,5,5,2,5,2,2,5]}}
    )
    Ribbon Width Mode=::NewEnumerator1
    UsePositionOffset=false
    Write Material Random=true
   ) {module=InitializeParticle001}
   &DynamicMaterialParameters(
    Index 0 Param 1=1
    Index 0 Param 2=5
    Index 0 Param 3=&SelectFloatFromArray(
     Direct Array Index=&SpawnGroup()
     Float Selection Array=DataInterface() {propertyValues {FloatData [10,4,4,10,4,10,10,4]}}
    )
    Index 0 Param 4=1
    Index 1 Param 1=&SelectFloatFromArray(
     Direct Array Index=&SpawnGroup()
     Float Selection Array=DataInterface() {propertyValues {FloatData [4,3,3,4,3,4,4,3]}}
    )
    Module.Index0Float0Write=true
    Module.Index0Float1Write=true
    Module.Index0Float2Write=true
    Module.Index0Float3Write=true
    Module.Index1Float0Write=true
    Module.Index1Float1Write=false
    Module.Index1Float2Write=false
    Module.Index1Float3Write=false
    Write Parameter Index 1=::NewEnumerator1
   )
  ]
  particleUpdateScript [
   &ParticleState(Kill Particles When Lifetime Has Elapsed=false)
   &SkeletalMeshLocation(
    Bone / Socket Index=&SelectIntFromArray(
     Direct Array Index=&ModuloInt(A=&SpawnGroup(), B=2)
     Int Selection Array=DataInterface() {propertyValues {IntData [2,3]}}
    )
    Bone Sampling Mode=::NewEnumerator4
    MaskSpawnGroup=false
    Orientation Sampling=::NewEnumerator1
    Sampled Position Offset=&SelectVectorFromArray(
     Direct Array Index=&ModuloInt(A=&SpawnGroup(), B=2)
     Direct Array Mode=::NewEnumerator1
     Vector Selection Array=DataInterface() {propertyValues {FloatData [<-1 -2 -3>, <1 -3 -3>]}}
    )
    Sampled Position Offset Coordinate Space:ENiagaraSkelMeshTransforms=::NewEnumerator2
    Skeletal Mesh=$System.Skeletal Mesh
    Use Sampled Mesh Attribute as Ribbon ID=false
   )
   /SetVariables(Particles.Position=&AddVectorToPosition(
     Position=$Particles.Position
     Vector=&Multiply_VectorByFloat(
      Vector=&Add_Vector(
       A=&Multiply_Vector(
        A=&NormalizeVector(
         Vector To Normalize=&FN_FrontEndVector3Modifier(Gameplay=$System.PlayerVelDir, Front End=$System.FrontendVelocity)
         Safe Normalize=true
         Safe Fallback Threshold=0
         Safe Fallback Vector=<0 0 0.001>
        )
        B=<0 0 -1>
       )
       B=&Multiply_VectorByFloat(
        Vector=&TransformVector(
         Vector=&SelectVectorFromArray(
          Direct Array Index=$Particles.SpawnGroup
          Direct Array Mode=::NewEnumerator1
          Vector Selection Array=DataInterface() {propertyValues {FloatData [<2 -1 0>, <-2 -0.5 0>, <1 -1.5 0>, <-0.5 -0.5 0>, <0.5 -1.5 0>, <-1 -1 0>, <1 -1 0>, <-2 -0.5 0>]}}
         )
         Destination Space=::Simulation
        )
        Float=&SelectFloatFromArray(
         Direct Array Index=&SpawnGroup()
         Float Selection Array=DataInterface() {propertyValues {FloatData [0.1,0.15,0.08,0.16,0.15,0.18]}}
        )
       )
      )
      Float=&FloatFromCurve(
       FloatCurve=DataInterface() {propertyValues {Curve {keys [{tangentMode=RCTM_User, time=0, value=0}, {tangentMode=RCTM_User, time=1, value=1}]}}}
       CurveIndex=$Particles.Normalized Spawn Group
       Scale Curve=&SelectFloatFromArray(
        Direct Array Index=&SpawnGroup()
        Float Selection Array=DataInterface() {propertyValues {FloatData [48,36,45,55,40,40,50,30]}}
       )
      )
     )
    )) {isSetParameters=true, module=SetVariables_#6}
   /SetVariables(Particles.Position=&AddVectorToPosition(
     Position=$Particles.Position
     Vector=&Multiply_VectorByFloat(
      Vector=/Niagara/DynamicInputs/Vectors/RotateVector(
       Vector To Rotate:Vector3f=<10 0 3>
       Yaw:NiagaraFloat=&Add_Float(
        A=$Particles.Age
        B=&Multiply_Float(A=&MakeFloatFromInt(Int=&SpawnGroup()), B=2)
       )
       Pitch:NiagaraFloat=0
       Roll:NiagaraFloat=0
       Rotation Rate:NiagaraFloat=&SelectFloatFromArray(
        Direct Array Index=&SpawnGroup()
        Float Selection Array=DataInterface() {propertyValues {FloatData [150,140,130,120,160,150,120,140]}}
       )
      )
      Float=&Multiply_Float(
       A=&Sine(
        Normalized Angle=&Add_Float(
         A=&Multiply_Float(
          A=$Emitter.Age
          B=&SelectFloatFromArray(
           Direct Array Index=&SpawnGroup()
           Float Selection Array=DataInterface() {propertyValues {FloatData [-2,-2.5,-2.5,-3,-3,-2,-2.5,-2]}}
          )
         )
         B=$Particles.Normalized Spawn Group
        )
        Period=0.9
        Scale=0.6
        Bias=0
       )
       B=&FloatFromCurve(
        FloatCurve=DataInterface() {propertyValues {Curve {keys [{tangentMode=RCTM_User, time=0, value=0}, {tangentMode=RCTM_User, time=1, value=1}]}}}
        CurveIndex=$Particles.Normalized Spawn Group
       )
      )
     )
    )) {isSetParameters=true, module=SetVariables_#7}
   /SetVariables(Particles.Position=&AddVectorToPosition(
     Position=$Particles.Position
     Vector=&Multiply_VectorByFloat(
      Vector=&SelectVectorFromArray(
       Direct Array Index=&SpawnGroup()
       Direct Array Mode=::NewEnumerator1
       Vector Selection Array=DataInterface() {propertyValues {FloatData [<15 0 -15>, <-15 0 -15>, <-25 0 0>, <30 0 0>]}}
      )
      Float=&FloatFromCurve(
       FloatCurve=DataInterface() {propertyValues {Curve {keys [
           {arriveTangent=4.1785, interpMode=RCIM_Cubic, leaveTangent=4.1785, tangentMode=RCTM_User, time=0, value=0}
           {arriveTangent=1.1179, interpMode=RCIM_Cubic, leaveTangent=1.1179, tangentMode=RCTM_User, time=0.2042, value=0.6228}
           {arriveTangent=0.0224, interpMode=RCIM_Cubic, leaveTangent=0.0224, tangentMode=RCTM_User, time=1, value=1}
          ]}}}
       CurveIndex=$Particles.Normalized Spawn Group
       Scale Curve=0.5
      )
     )
    )) {isSetParameters=true}
   &FN_Sprite_Attributes("Create Normalized Age "=true)
   &ScaleColor(
    Scale RGB=&VectorFromFloat(Value=&FNTimeOfDayScalar(Night Scale=0.8, Morning Scale=0.9, Evening Scale=0.9))
    ScaleA=true
    ScaleRGB=true
   )
  ]
  rendererClasses [&NiagaraRibbonRendererProperties]
  renderers [{rendererClass=&NiagaraRibbonRendererProperties}]
 }
 Chains_Hands {
  emitterSpawnScript [&FN_PreserveSkydiveUserVariables() {module=PreserveUserVariables}]
  emitterUpdateScript [
   &EmitterState(
    Inactive Response=::NewEnumerator1
    Life Cycle Mode=::NewEnumerator1
    Loop Behavior=::NewEnumerator1
    Loop Duration Mode=::NewEnumerator1
    UseLoopDelay=false
   )
   /SetVariables(Emitter.Number of Chains=8, Emitter.Particle per Chain=9) {isSetParameters=true, module=SetVariables_#5}
   &SpawnBurst_Instantaneous(
    Spawn Count=&Multiply_Int(A=$Emitter.Number of Chains, B=$Emitter.Particle per Chain)
    Spawn Probability=HlslExpression() {hlslExpression="Emitter.LoopCount == 0"}
    Use Spawn Probability=false
   ) {module=SpawnBurst_Instantaneous001}
  ]
  particleSpawnScript [
   &FN_SetSpawnGroup(
    Inclusivity=true
    Normalize Spawn Group=true
    Particles Per Group:NiagaraInt32=$Emitter.Particle per Chain
    Pattern=::NewEnumerator0
    Set Ribbon ID and Ribbon Link Order=true
   )
   &InitializeParticle(
    Ribbon Width=&SelectFloatFromArray(
     Direct Array Index=&SpawnGroup()
     Float Selection Array=DataInterface() {propertyValues {FloatData [6,6,6,6,2,2,2,2]}}
    )
    Ribbon Width Mode=::NewEnumerator1
    UsePositionOffset=false
    Write Material Random=true
   ) {module=InitializeParticle001}
   &DynamicMaterialParameters(
    Index 0 Param 1=1
    Index 0 Param 2=5
    Index 0 Param 3=&Multiply_Float(
     A=&SelectFloatFromArray(
      Direct Array Index=&SpawnGroup()
      Float Selection Array=DataInterface() {propertyValues {FloatData [4.07,4.07,4.07,4.07,20.13,20,20.13,20.13]}}
     )
     B=&FloatFromCurve(
      FloatCurve=DataInterface() {propertyValues {Curve {keys [<at 0 1 RCIM_Cubic>,<at 1 1.2>]}}}
      CurveIndex=$Particles.RibbonLinkOrder
      Scale Curve=&SelectFloatFromArray(
       Direct Array Index=&SpawnGroup()
       Float Selection Array=DataInterface() {propertyValues {FloatData [1,1,1,1,0.8,0.8,1,1]}}
      )
     )
    )
    Index 0 Param 4=1
    Index 1 Param 1=&SelectFloatFromArray(
     Direct Array Index=&SpawnGroup()
     Float Selection Array=DataInterface() {propertyValues {FloatData [1,1,1,1,4,4,4,4]}}
    )
    Module.Index0Float0Write=true
    Module.Index0Float1Write=true
    Module.Index0Float2Write=true
    Module.Index0Float3Write=true
    Module.Index1Float0Write=true
    Module.Index1Float1Write=false
    Module.Index1Float2Write=false
    Module.Index1Float3Write=false
    Write Parameter Index 1=::NewEnumerator1
   )
  ]
  particleUpdateScript [
   &ParticleState(Kill Particles When Lifetime Has Elapsed=false)
   &SkeletalMeshLocation(
    Bone / Socket Index=&ModuloInt(A=&SpawnGroup(), B=2)
    Bone Sampling Mode=::NewEnumerator4
    MaskSpawnGroup=false
    Orientation Sampling=::NewEnumerator1
    Sampled Position Offset=<0 0 -4>
    Sampled Position Offset Coordinate Space:ENiagaraSkelMeshTransforms=::NewEnumerator2
    Skeletal Mesh=$System.Skeletal Mesh
    Use Sampled Mesh Attribute as Ribbon ID=false
   )
   /SetVariables(Particles.Position=&AddVectorToPosition(
     Position=$Particles.Position
     Vector=&Multiply_VectorByFloat(
      Vector=&Add_Vector(
       A=&Multiply_Vector(
        A=&NormalizeVector(
         Vector To Normalize=&FN_FrontEndVector3Modifier(Gameplay=$System.PlayerVelDir, Front End=$System.FrontendVelocity)
         Safe Normalize=true
         Safe Fallback Threshold=0
         Safe Fallback Vector=<0 0 0.001>
        )
        B=&SelectVectorFromArray(
         Direct Array Index=$Particles.SpawnGroup
         Vector Selection Array=DataInterface() {propertyValues {FloatData [<-0.5 -1 -1>, <-0.5 -0.5 -1>, <0.1 -1.5 -1>, <0.1 -0.5 -1>, <-0.5 -1.5 -1>, <-0.5 -1 -1>, <0.1 -1 -1>, <0.1 -0.5 -1>]}}
        )
       )
       B=&Multiply_VectorByFloat(
        Vector=&TransformVector(
         Vector=&SelectVectorFromArray(
          Direct Array Index=$Particles.SpawnGroup
          Direct Array Mode=::NewEnumerator1
          Vector Selection Array=DataInterface() {propertyValues {FloatData [<2 -2 0>, <-2 0 0>, <1 2 0>, <-0.5 0 0>, <0.5 0 0>, <-1 2 0>, <1 1 0>, <-2 1 0>]}}
         )
         Destination Space=::Simulation
        )
        Float=&SelectFloatFromArray(
         Direct Array Index=&SpawnGroup()
         Float Selection Array=DataInterface() {propertyValues {FloatData [0.1,0.15,0.08,0.16,0.15,0.18]}}
        )
       )
      )
      Float=&FloatFromCurve(
       FloatCurve=DataInterface() {propertyValues {Curve {keys [{tangentMode=RCTM_User, time=0, value=0}, {tangentMode=RCTM_User, time=1, value=1}]}}}
       CurveIndex=$Particles.Normalized Spawn Group
       Scale Curve=&SelectFloatFromArray(
        Direct Array Index=&SpawnGroup()
        Float Selection Array=DataInterface() {propertyValues {FloatData [105,97,120,100,75,55,94,80]}}
       )
      )
     )
    )) {isSetParameters=true, module=SetVariables_#6}
   /SetVariables(Particles.Position=&AddVectorToPosition(
     Position=$Particles.Position
     Vector=&Multiply_VectorByFloat(
      Vector=&RotateVector(
       Vector To Rotate:Vector3f=<10 0 3>
       Yaw:NiagaraFloat=&Add_Float(
        A=$Particles.Age
        B=&Multiply_Float(A=&MakeFloatFromInt(Int=&SpawnGroup()), B=2)
       )
       Pitch:NiagaraFloat=0
       Roll:NiagaraFloat=0
       Rotation Rate:NiagaraFloat=&SelectFloatFromArray(
        Direct Array Index=&SpawnGroup()
        Float Selection Array=DataInterface() {propertyValues {FloatData [150,140,130,120,160,150,120,140]}}
       )
      )
      Float=&Multiply_Float(
       A=&Sine(
        Normalized Angle=&Add_Float(
         A=&Multiply_Float(
          A=$Emitter.Age
          B=&SelectFloatFromArray(
           Direct Array Index=&SpawnGroup()
           Float Selection Array=DataInterface() {propertyValues {FloatData [-2,-2.5,-2.5,-3,-3,-2,-2.5,-2]}}
          )
         )
         B=$Particles.Normalized Spawn Group
        )
        Period=0.9
        Scale=1
        Bias=0
       )
       B=&FloatFromCurve(
        FloatCurve=DataInterface() {propertyValues {Curve {keys [{tangentMode=RCTM_User, time=0, value=0}, {tangentMode=RCTM_User, time=1, value=1}]}}}
        CurveIndex=$Particles.Normalized Spawn Group
       )
      )
     )
    )) {isSetParameters=true, module=SetVariables_#7}
   /SetVariables(Particles.Position=&AddVectorToPosition(
     Position=$Particles.Position
     Vector=&Multiply_VectorByFloat(
      Vector=&SelectVectorFromArray(
       Direct Array Index=&SpawnGroup()
       Direct Array Mode=::NewEnumerator1
       Vector Selection Array=DataInterface() {propertyValues {FloatData [<25 0 0>, <-25 0 0>, <-25 0 0>, <25 0 0>]}}
      )
      Float=&FloatFromCurve(
       FloatCurve=DataInterface() {propertyValues {Curve {keys [
           {arriveTangent=4.1785, interpMode=RCIM_Cubic, leaveTangent=4.1785, tangentMode=RCTM_User, time=0, value=0}
           {arriveTangent=1.1179, interpMode=RCIM_Cubic, leaveTangent=1.1179, tangentMode=RCTM_User, time=0.2042, value=0.6228}
           {arriveTangent=0.0224, interpMode=RCIM_Cubic, leaveTangent=0.0224, tangentMode=RCTM_User, time=1, value=1}
          ]}}}
       CurveIndex=$Particles.Normalized Spawn Group
      )
     )
    )) {isSetParameters=true}
   &FN_Sprite_Attributes("Create Normalized Age "=true)
   &ScaleColor(
    Scale RGB=&VectorFromFloat(Value=&FNTimeOfDayScalar(Night Scale=0.8, Morning Scale=0.9, Evening Scale=0.9))
    ScaleA=true
    ScaleRGB=true
   )
   &SolveForcesAndVelocity(Clamp Velocity=false, Limit Acceleration=false) {enabled=false}
  ]
  rendererClasses [&NiagaraRibbonRendererProperties]
  renderers [{rendererClass=&NiagaraRibbonRendererProperties}]
 }
 Chains_Hands_Small {
  emitterSpawnScript [&FN_PreserveSkydiveUserVariables() {module=PreserveUserVariables}]
  emitterUpdateScript [
   &EmitterState(
    Inactive Response=::NewEnumerator1
    Life Cycle Mode=::NewEnumerator1
    Loop Behavior=::NewEnumerator1
    Loop Duration Mode=::NewEnumerator1
    UseLoopDelay=false
   )
   /SetVariables(Emitter.Number of Chains=4, Emitter.Particle per Chain=5) {isSetParameters=true, module=SetVariables_#5}
   &SpawnBurst_Instantaneous(
    Spawn Count=&Multiply_Int(A=$Emitter.Number of Chains, B=$Emitter.Particle per Chain)
    Spawn Probability=HlslExpression() {hlslExpression="Emitter.LoopCount == 0"}
    Use Spawn Probability=false
   ) {module=SpawnBurst_Instantaneous001}
  ]
  particleSpawnScript [
   &FN_SetSpawnGroup(
    Inclusivity=true
    Normalize Spawn Group=true
    Particles Per Group:NiagaraInt32=$Emitter.Particle per Chain
    Pattern=::NewEnumerator0
    Set Ribbon ID and Ribbon Link Order=true
   )
   &InitializeParticle(
    Ribbon Width=&SelectFloatFromArray(Direct Array Index=&SpawnGroup(), Float Selection Array=DataInterface() {propertyValues {FloatData [4,6,6,4]}})
    Ribbon Width Mode=::NewEnumerator1
    UsePositionOffset=false
    Write Material Random=true
   ) {module=InitializeParticle001}
   &DynamicMaterialParameters(
    Index 0 Param 1=1
    Index 0 Param 2=5
    Index 0 Param 3=&SelectFloatFromArray(Direct Array Index=&SpawnGroup(), Float Selection Array=DataInterface() {propertyValues {FloatData [5,4,4,5]}})
    Index 0 Param 4=1
    Index 1 Param 1=&SelectFloatFromArray(Direct Array Index=&SpawnGroup(), Float Selection Array=DataInterface() {propertyValues {FloatData [2,3,3,2]}})
    Module.Index0Float0Write=true
    Module.Index0Float1Write=true
    Module.Index0Float2Write=true
    Module.Index0Float3Write=true
    Module.Index1Float0Write=true
    Module.Index1Float1Write=false
    Module.Index1Float2Write=false
    Module.Index1Float3Write=false
    Write Parameter Index 1=::NewEnumerator1
   )
  ]
  particleUpdateScript [
   &ParticleState(Kill Particles When Lifetime Has Elapsed=false)
   &SkeletalMeshLocation(
    Bone / Socket Index=&ModuloInt(A=&SpawnGroup(), B=2)
    Bone Sampling Mode=::NewEnumerator4
    MaskSpawnGroup=false
    Orientation Sampling=::NewEnumerator1
    Sampled Position Offset=&SelectVectorFromArray(
     Direct Array Index=&ModuloInt(A=&SpawnGroup(), B=2)
     Direct Array Mode=::NewEnumerator1
     Vector Selection Array=DataInterface() {propertyValues {FloatData [<-1 3 -3>, <1 3 -3>]}}
    )
    Sampled Position Offset Coordinate Space:ENiagaraSkelMeshTransforms=::NewEnumerator2
    Skeletal Mesh=$System.Skeletal Mesh
    Use Sampled Mesh Attribute as Ribbon ID=false
   )
   /SetVariables(Particles.Position=&AddVectorToPosition(
     Position=$Particles.Position
     Vector=&Multiply_VectorByFloat(
      Vector=&Add_Vector(
       A=&Multiply_Vector(
        A=&NormalizeVector(
         Vector To Normalize=&FN_FrontEndVector3Modifier(Gameplay=$System.PlayerVelDir, Front End=$System.FrontendVelocity)
         Safe Normalize=true
         Safe Fallback Threshold=0
         Safe Fallback Vector=<0 0 0.001>
        )
        B=<0 0 -1>
       )
       B=&Multiply_VectorByFloat(
        Vector=&TransformVector(
         Vector=&SelectVectorFromArray(
          Direct Array Index=$Particles.SpawnGroup
          Direct Array Mode=::NewEnumerator1
          Vector Selection Array=DataInterface() {propertyValues {FloatData [<1.2 -1 0>, <-1.3 -1 0>, <1 -1 0>, <0.5 -1 0>]}}
         )
         Destination Space=::Simulation
        )
        Float=&SelectFloatFromArray(
         Direct Array Index=&SpawnGroup()
         Float Selection Array=DataInterface() {propertyValues {FloatData [0.5,0.25,0.34,0.22]}}
        )
       )
      )
      Float=&FloatFromCurve(
       FloatCurve=DataInterface() {propertyValues {Curve {keys [{tangentMode=RCTM_User, time=0, value=0}, {tangentMode=RCTM_User, time=1, value=1}]}}}
       CurveIndex=$Particles.Normalized Spawn Group
       Scale Curve=&SelectFloatFromArray(
        Direct Array Index=&SpawnGroup()
        Float Selection Array=DataInterface() {propertyValues {FloatData [25,30,25,22]}}
       )
      )
     )
    )) {isSetParameters=true, module=SetVariables_#6}
   /SetVariables(Particles.Position=&AddVectorToPosition(
     Position=$Particles.Position
     Vector=&Multiply_VectorByFloat(
      Vector=&RotateVector(
       Vector To Rotate:Vector3f=<10 0 3>
       Yaw:NiagaraFloat=&Add_Float(
        A=$Particles.Age
        B=&Multiply_Float(A=&MakeFloatFromInt(Int=&SpawnGroup()), B=2)
       )
       Pitch:NiagaraFloat=0
       Roll:NiagaraFloat=0
       Rotation Rate:NiagaraFloat=&SelectFloatFromArray(
        Direct Array Index=&SpawnGroup()
        Float Selection Array=DataInterface() {propertyValues {FloatData [150,140,130,120]}}
       )
      )
      Float=&Multiply_Float(
       A=&Sine(
        Normalized Angle=&Add_Float(
         A=&Multiply_Float(
          A=$Emitter.Age
          B=&SelectFloatFromArray(
           Direct Array Index=&SpawnGroup()
           Float Selection Array=DataInterface() {propertyValues {FloatData [-1.5,-1.875,-1.5,-1.875]}}
          )
         )
         B=$Particles.Normalized Spawn Group
        )
        Period=1
        Scale=&Lerp_Float(A=0.8, B=1, Alpha=$System.Dive)
        Bias=0
       )
       B=&FloatFromCurve(
        FloatCurve=DataInterface() {propertyValues {Curve {keys [{tangentMode=RCTM_User, time=0, value=0}, {tangentMode=RCTM_User, time=1, value=1}]}}}
        CurveIndex=$Particles.Normalized Spawn Group
       )
      )
     )
    )) {isSetParameters=true, module=SetVariables_#7}
   /SetVariables(Particles.Position=&AddVectorToPosition(
     Position=$Particles.Position
     Vector=&Multiply_VectorByFloat(
      Vector=&SelectVectorFromArray(
       Direct Array Index=&SpawnGroup()
       Direct Array Mode=::NewEnumerator1
       Vector Selection Array=DataInterface() {propertyValues {FloatData [<1 0 0>, <-1 0 0>]}}
      )
      Float=&FloatFromCurve(
       FloatCurve=DataInterface() {propertyValues {Curve {keys [
           {arriveTangent=4.1785, interpMode=RCIM_Cubic, leaveTangent=4.1785, tangentMode=RCTM_User, time=0, value=0}
           {arriveTangent=1.1179, interpMode=RCIM_Cubic, leaveTangent=1.1179, tangentMode=RCTM_User, time=0.2042, value=0.6228}
           {arriveTangent=0.0224, interpMode=RCIM_Cubic, leaveTangent=0.0224, tangentMode=RCTM_User, time=1, value=1}
          ]}}}
       CurveIndex=$Particles.Normalized Spawn Group
       Scale Curve=10
      )
     )
    )) {isSetParameters=true}
   &FN_Sprite_Attributes("Create Normalized Age "=true)
   &ScaleColor(
    Scale RGB=&VectorFromFloat(Value=&FNTimeOfDayScalar(Night Scale=0.8, Morning Scale=0.9, Evening Scale=0.9))
    ScaleA=true
    ScaleRGB=true
   )
  ]
  rendererClasses [&NiagaraRibbonRendererProperties]
  renderers [{rendererClass=&NiagaraRibbonRendererProperties}]
 }
 CrossHands_Chains {
  emitterSpawnScript [&FN_PreserveSkydiveUserVariables()]
  emitterUpdateScript [
   &EmitterState(
    Inactive Response=::NewEnumerator1
    Life Cycle Mode=::NewEnumerator1
    Loop Behavior=::NewEnumerator1
    Loop Duration Mode=::NewEnumerator1
    UseLoopDelay=false
   )
   /SetVariables(Emitter.Number of Chains=5, Emitter.Particle per Chain=25) {isSetParameters=true, module=SetVariables_#5}
   &SpawnBurst_Instantaneous(
    Spawn Count=&Multiply_Int(A=$Emitter.Number of Chains, B=$Emitter.Particle per Chain)
    Use Spawn Probability=false
    UseLoopCountLimit=false
   )
   &SampleSkeletalMesh(
    AdditionalYawPitchRoll=false
    Bone Sampling Mode:ENiagaraBoneSamplingMode=::NewEnumerator4
    Position Offset Coordinate Space:ENiagaraCoordinateSpace=::Local
    Sampled Position Offset:Vector3f=<-1 5 -5>
    Skeletal Mesh:NiagaraDataInterfaceSkeletalMesh=$System.Skeletal Mesh
   )
   &SampleSkeletalMesh(
    AdditionalYawPitchRoll=false
    Bone / Socket Index=1
    Bone Sampling Mode:ENiagaraBoneSamplingMode=::NewEnumerator4
    Position Offset Coordinate Space:ENiagaraCoordinateSpace=::Local
    Sampled Position Offset:Vector3f=<1 5 -5>
    Skeletal Mesh:NiagaraDataInterfaceSkeletalMesh=$System.Skeletal Mesh
   ) {module=SampleSkeletalMesh001}
  ]
  particleSpawnScript [
   &FN_SetSpawnGroup(
    Inclusivity=true
    Normalize Spawn Group=true
    Particles Per Group:NiagaraInt32=$Emitter.Particle per Chain
    Pattern=::NewEnumerator0
    Set Ribbon ID and Ribbon Link Order=true
   )
   &InitializeRibbon(
    Ribbon Width=&SelectFloatFromArray(
     Direct Array Index=&SpawnGroup()
     Float Selection Array=DataInterface() {propertyValues {FloatData [4,6,2,6,2]}}
    )
    UsePositionOffset=false
    Write Material Random=true
   )
   &DynamicMaterialParameters(
    Index 0 Param 1=1
    Index 0 Param 2=5
    Index 0 Param 3=&SelectFloatFromArray(
     Direct Array Index=&SpawnGroup()
     Float Selection Array=DataInterface() {propertyValues {FloatData [10,12,25,4,20]}}
    )
    Index 0 Param 4=1
    Index 1 Param 1=&SelectFloatFromArray(
     Direct Array Index=&SpawnGroup()
     Float Selection Array=DataInterface() {propertyValues {FloatData [2,3,4,1,0]}}
    )
    Module.Index0Float0Write=true
    Module.Index0Float1Write=true
    Module.Index0Float2Write=true
    Module.Index0Float3Write=true
    Module.Index1Float0Write=true
    Module.Index1Float1Write=false
    Module.Index1Float2Write=false
    Module.Index1Float3Write=false
    Write Parameter Index 1=::NewEnumerator1
   )
  ]
  particleUpdateScript [
   &ParticleState(Kill Particles When Lifetime Has Elapsed=false, Let Infinitely Lived Particles Die When Emitter Deactivates=true)
   /SetVariables(Particles.float:NiagaraFloat=&FloatFromCurve(FloatCurve=DataInterface() {propertyValues {Curve {keys [<at 0.6 0 RCIM_Cubic>,<at 1 1 RCIM_Cubic>]}}}, CurveIndex=$User.ParaGlideLeanAlpha)) {isSetParameters=true}
   /Game/Effects/Niagara/Modules/Curves/FN_HermiteSpline(
    Direction End:Vector3f=&Add_Vector(
     A=&Add_Vector(
      A=&Multiply_VectorByFloat(
       Vector=&SubtractVectorFromPosition(
        Position=&SelectVectorFromArray(
         Direct Array Index=&SpawnGroup()
         Vector Selection Array=DataInterface() {propertyValues {FloatData [<1 0.5 0.1>, <1 0.5 1>, <1 0.7 0.6>, <1 0.1 0.5>, <1 0.4 0.8>, <1 1.1 0.7>]}}
        )
        Vector=&Multiply_Vector(
         A=&NormalizeVector(Vector To Normalize=$System.PlayerVelDir, Safe Normalize=true, Safe Fallback Threshold=0, Safe Fallback Vector=<0 0 0.001>)
         B=&AddVectorToPosition(
          Position=&Multiply_Vector(
           A=&NormalizeVector(Vector To Normalize=$System.PlayerVelDir, Safe Normalize=true, Safe Fallback Threshold=0, Safe Fallback Vector=<0 0 0.001>)
           B=<-1 0 0>
          )
          Vector=&Lerp_Vector(A=<0 0 0>, B=<0 -0.5 0>, Alpha=$User.ParaGlideLeanAlpha)
         )
        )
       )
       Float=&SelectFloatFromArray(
        Direct Array Index=&SpawnGroup()
        Float Selection Array=DataInterface() {propertyValues {FloatData [75,175,111,122,133,144]}}
       )
      )
      B=&Multiply_Vector(
       A=&NormalizeVector(
        Vector To Normalize=&FN_FrontEndVector3Modifier(
         Gameplay=&TransformVector(Vector=$Engine.Owner.Velocity, Source Space=::World, Destination Space=::Simulation)
         Front End=$System.FrontendVelocity
        )
        Safe Normalize=false
        Safe Fallback Threshold=0
        Safe Fallback Vector=<0 0 0.001>
       )
       B=&FN_FrontEndVector3Modifier(
        Gameplay=<0 0 -100>
        Front End=&Lerp_Vector(A=<-100 -100 -100>, B=<-200 -200 -200>, Alpha=$User.ParaGlideLeanAlpha)
       )
      )
     )
     B=&Multiply_Vector(
      A=&NormalizeVector(
       Vector To Normalize=&Lerp_Vector(A=<0 0 1>, B=<1 1 1>, Alpha=$Particles.float)
       Safe Normalize=true
       Safe Fallback Threshold=0
       Safe Fallback Vector=<0 0 0.001>
      )
      B=&Lerp_Vector(
       A=<0 0 0>
       B=&FN_FrontEndVector3Modifier(Gameplay=<0 0 100>, Front End=<0 0 0>)
       Alpha=$Particles.float
      )
     )
    )
    Direction Start:Vector3f=&Add_Vector(
     A=&Add_Vector(
      A=&Multiply_VectorByFloat(
       Vector=&SubtractVectorFromPosition(
        Position=&SelectVectorFromArray(
         Direct Array Index=&SpawnGroup()
         Vector Selection Array=DataInterface() {propertyValues {FloatData [<-1 0.5 0.1>, <-1 0.5 1>, <-1 0.7 0.6>, <-1 0.1 0.5>, <-1 0.4 0.8>, <-1 1.1 0.7>]}}
        )
        Vector=&AddVectorToPosition(
         Position=&Multiply_Vector(
          A=&NormalizeVector(Vector To Normalize=$System.PlayerVelDir, Safe Normalize=true, Safe Fallback Threshold=0, Safe Fallback Vector=<0 0 0.001>)
          B=<-1 0 0>
         )
         Vector=&Lerp_Vector(A=<0 0 0>, B=<0 -0.5 0>, Alpha=$User.ParaGlideLeanAlpha)
        )
       )
       Float=&SelectFloatFromArray(
        Direct Array Index=&SpawnGroup()
        Float Selection Array=DataInterface() {propertyValues {FloatData [75,175,111,122,133,144]}}
       )
      )
      B=&Multiply_Vector(
       A=&NormalizeVector(
        Vector To Normalize=&FN_FrontEndVector3Modifier(
         Gameplay=&TransformVector(Vector=$Engine.Owner.Velocity, Source Space=::World, Destination Space=::Simulation)
         Front End=$System.FrontendVelocity
        )
        Safe Normalize=true
        Safe Fallback Threshold=0
        Safe Fallback Vector=<0 0 0.001>
       )
       B=&FN_FrontEndVector3Modifier(
        Gameplay=<0 0 -100>
        Front End=&Lerp_Vector(A=<-100 -100 -100>, B=<-200 -200 -200>, Alpha=$User.ParaGlideLeanAlpha)
       )
      )
     )
     B=&Multiply_Vector(
      A=&NormalizeVector(
       Vector To Normalize=&Lerp_Vector(A=<0 0 1>, B=<1 1 1>, Alpha=$Particles.float)
       Safe Normalize=true
       Safe Fallback Threshold=0
       Safe Fallback Vector=<0 0 0.001>
      )
      B=&Lerp_Vector(
       A=<0 0 0>
       B=&FN_FrontEndVector3Modifier(Gameplay=<0 0 100>, Front End=<0 0 0>)
       Alpha=$Particles.float
      )
     )
    )
    Position Blend:NiagaraFloat=$Particles.Normalized Spawn Group
    Position End:Vector3f=&AddVectorToPosition(
     Position=&AddVectorToPosition(
      Position=$Emitter.SampleSkeletalMesh001.SampledPosition
      Vector=&Multiply_VectorByFloat(
       Vector=<5 0 0>
       Float=&FloatFromCurve(
        FloatCurve=DataInterface() {propertyValues {Curve {keys [{arriveTangent=4.1785, time=0.9, value=0}, <at 0.95 1 RCIM_Cubic>, <at 1 0 RCIM_Cubic>]}}}
        CurveIndex=$Particles.Normalized Spawn Group
       )
      )
     )
     Vector=&RotateVectorByQuaternion(Vector To Rotate=<-2 -2 0>, Quaternion=$Emitter.SampleSkeletalMesh001.SampledRotation)
    )
    Position Start:Vector3f=&AddVectorToPosition(
     Position=&AddVectorToPosition(
      Position=$Emitter.SampleSkeletalMesh.SampledPosition
      Vector=&Multiply_VectorByFloat(
       Vector=<-5 0 0>
       Float=&FloatFromCurve(
        FloatCurve=DataInterface() {propertyValues {Curve {keys [
            {arriveTangent=4.1785, interpMode=RCIM_Cubic, leaveTangent=4.1785, tangentMode=RCTM_User, time=0, value=0}
            {arriveTangent=0.0224, leaveTangent=0.0224, time=0.05, value=1}
            <at 0.1 0 RCIM_Cubic>
           ]}}}
        CurveIndex=$Particles.Normalized Spawn Group
       )
      )
     )
     Vector=&RotateVectorByQuaternion(Vector To Rotate=<2 2 0>, Quaternion=$Emitter.SampleSkeletalMesh.SampledRotation)
    )
    Write Particles Position=true
   )
   &ScaleRibbonWidth(Ribbon Width Scale=&FloatFromCurve(
     FloatCurve=DataInterface() {propertyValues {Curve {keys [
         {interpMode=RCIM_Cubic, tangentMode=RCTM_User, time=0, value=0.7}
         {interpMode=RCIM_Cubic, tangentMode=RCTM_User, time=0.5, value=1}
         {tangentMode=RCTM_User, time=1, value=0.7}
        ]}}}
     CurveIndex=$Particles.Normalized Spawn Group
    ))
   &PendulumSetup(
    Module.Enable Spring Driver=false
    Pendulum Length:NiagaraFloat=&FloatFromCurve(
     FloatCurve=DataInterface() {propertyValues {Curve {keys [
         {arriveTangent=3.74, interpMode=RCIM_Cubic, leaveTangent=3.74, tangentMode=RCTM_User, time=0, value=0}
         {interpMode=RCIM_Cubic, tangentMode=RCTM_User, time=0.5, value=1}
         {arriveTangent=-4.8628, interpMode=RCIM_Cubic, leaveTangent=-4.8628, tangentMode=RCTM_User, time=1, value=0}
        ]}}}
     CurveIndex=$Particles.Normalized Spawn Group
     Scale Curve=2
    )
    Pendulum Pivot:NiagaraPosition=$Particles.Position
    "Spring Driven Constraint (Tightness)"=500
   )
   &CurlNoiseForce(
    Noise Frequency=35
    Noise Strength=&Lerp_Float(A=150000, B=25000, Alpha=$System.Dive)
    Pan Noise=true
    Pan Noise Field=<0 0 7>
   )
   &SolveForcesAndVelocity(Clamp Velocity=false, Limit Acceleration=false)
   &PendulumConstraint(Calculate and Output Potential Energy=true)
   &CalculateAccurateVelocity()
   &ScaleColor(
    Scale RGB=&VectorFromFloat(Value=&FNTimeOfDayScalar(Night Scale=0.8, Morning Scale=0.9, Evening Scale=0.9))
    ScaleA=true
    ScaleRGB=true
   ) {module=ScaleColor001}
  ]
  rendererClasses [&NiagaraRibbonRendererProperties]
  renderers [{rendererClass=&NiagaraRibbonRendererProperties}]
 }
 Glints_CrossHands {
  emitterUpdateScript [
   &EmitterState(
    Inactive Response=::NewEnumerator1
    Life Cycle Mode=::NewEnumerator1
    Loop Behavior=::NewEnumerator1
    Loop Duration Mode=::NewEnumerator1
    UseLoopDelay=false
   )
   &SpawnRate(SpawnRate=10, Use Spawn Probability=false)
  ]
  particleSpawnScript [
   &InitializeParticle(
    Color=&RandomRangeLinearColor(Minimum=<rgba 10 7.0268 3.8021 0>, Maximum=<rgba 10 4.375 2.5 0.7>, Color Channel Mode=::NewEnumerator2)
    Lifetime Max=1
    Lifetime Min=0.5
    Lifetime Mode=::NewEnumerator1
    Sprite Rotation Mode=::NewEnumerator2
    Sprite Size Mode=::NewEnumerator1
    Sprite UV Mode=::NewEnumerator3
    Uniform Sprite Size Max=15
    UsePositionOffset=false
    Write Material Random=true
   )
   &CameraOffset(Camera Offset Amount=5)
   /SetVariables(
    Particles.RandFloat:NiagaraFloat=&RandomRangeFloat(Minimum=0, Maximum=1)
    Particles.RandForAttr:NiagaraInt32=&RandomRangeInt(Minimum=0, Maximum=2)
    Particles.RandInt:NiagaraInt32=&RandomRangeInt(Minimum=0, Maximum=239)
   ) {isSetParameters=true, module=SetVariables_#8}
  ]
  particleUpdateScript [
   &ParticleState()
   &self:SetRandomPositionOnChain(
    GroupNum=5
    Particle Attribute Reader=DataInterface() {propertyValues {EmitterBinding {bindingMode=Other, emitterName=CrossHands_Chains}}}
    ParticleNum=125
    RandFloat:NiagaraFloat=$Particles.RandFloat
    RandInt:NiagaraInt32=$Particles.RandInt
   ) {module=SetRandomPositionOnChain}
   &ScaleSpriteSize(Uniform Curve Sprite Scale=DataInterface() {propertyValues {Curve {keys [
        {interpMode=RCIM_Cubic, tangentMode=RCTM_User, time=0, value=0}
        {interpMode=RCIM_Cubic, tangentMode=RCTM_User, time=0.5, value=1}
        {tangentMode=RCTM_User, time=1, value=0}
       ]}}})
   &SpriteRotationRate(Rotation Rate=&RandomRangeFloat(Minimum=200, Maximum=360))
   &JitterPosition(Jitter Amount=1.5, Jitter Delay=0.25)
   &ScaleColor(
    Scale RGB=&VectorFromFloat(Value=&FNTimeOfDayScalar(Night Scale=0.5, Morning Scale=0.75, Evening Scale=0.75))
    ScaleA=true
    ScaleRGB=true
   )
  ]
  rendererClasses [&NiagaraSpriteRendererProperties]
  renderers [{rendererClass=&NiagaraSpriteRendererProperties}]
 }
 Glints_Feet {
  emitterUpdateScript [
   &EmitterState(
    Inactive Response=::NewEnumerator1
    Life Cycle Mode=::NewEnumerator1
    Loop Behavior=::NewEnumerator1
    Loop Duration Mode=::NewEnumerator1
    UseLoopDelay=false
   )
   &SpawnRate(SpawnRate=10, Use Spawn Probability=false)
  ]
  particleSpawnScript [
   &InitializeParticle(
    Color=&RandomRangeLinearColor(Minimum=<rgba 10 7.0268 3.8021 0>, Maximum=<rgba 10 4.375 2.5 0.7>, Color Channel Mode=::NewEnumerator2)
    Lifetime Max=1
    Lifetime Min=0.5
    Lifetime Mode=::NewEnumerator1
    Sprite Rotation Mode=::NewEnumerator2
    Sprite Size Mode=::NewEnumerator1
    Sprite UV Mode=::NewEnumerator3
    Uniform Sprite Size Min=2.5
    UsePositionOffset=false
    Write Material Random=true
   )
   &CameraOffset(Camera Offset Amount=5)
   /SetVariables(
    Particles.RandFloat:NiagaraFloat=&RandomRangeFloat(Minimum=0, Maximum=1)
    Particles.RandForAttr:NiagaraInt32=&RandomRangeInt(Minimum=0, Maximum=2)
    Particles.RandInt:NiagaraInt32=&RandomRangeInt(Minimum=0, Maximum=239)
   ) {isSetParameters=true, module=SetVariables_#8}
  ]
  particleUpdateScript [
   &ParticleState()
   &self:SetRandomPositionOnChain(
    GroupNum=8
    Particle Attribute Reader=DataInterface() {propertyValues {EmitterBinding {bindingMode=Other, emitterName=Chains_Feet}}}
    ParticleNum=40
    RandFloat:NiagaraFloat=$Particles.RandFloat
    RandInt:NiagaraInt32=$Particles.RandInt
   ) {module=SetRandomPositionOnChain}
   &ScaleSpriteSize(Uniform Curve Sprite Scale=DataInterface() {propertyValues {Curve {keys [
        {interpMode=RCIM_Cubic, tangentMode=RCTM_User, time=0, value=0}
        {interpMode=RCIM_Cubic, tangentMode=RCTM_User, time=0.5, value=1}
        {tangentMode=RCTM_User, time=1, value=0}
       ]}}})
   &SpriteRotationRate(Rotation Rate=&RandomRangeFloat(Minimum=200, Maximum=360))
   &JitterPosition(Jitter Amount=1, Jitter Delay=0.25)
   &ScaleColor(
    Scale RGB=&VectorFromFloat(Value=&FNTimeOfDayScalar(Night Scale=0.5, Morning Scale=0.75, Evening Scale=0.75))
    ScaleA=true
    ScaleRGB=true
   )
  ]
  rendererClasses [&NiagaraSpriteRendererProperties]
  renderers [{rendererClass=&NiagaraSpriteRendererProperties}]
 }
 Glints_Hands {
  emitterUpdateScript [
   &EmitterState(
    Inactive Response=::NewEnumerator1
    Life Cycle Mode=::NewEnumerator1
    Loop Behavior=::NewEnumerator1
    Loop Duration Mode=::NewEnumerator1
    UseLoopDelay=false
   )
   &SpawnRate(SpawnRate=10, Use Spawn Probability=false)
  ]
  particleSpawnScript [
   &InitializeParticle(
    Color=&RandomRangeLinearColor(Minimum=<rgba 10 7.0268 3.8021 0>, Maximum=<rgba 10 4.375 2.5 0.7>, Color Channel Mode=::NewEnumerator2)
    Lifetime Max=1
    Lifetime Min=0.5
    Lifetime Mode=::NewEnumerator1
    Sprite Rotation Mode=::NewEnumerator2
    Sprite Size Mode=::NewEnumerator1
    Sprite UV Mode=::NewEnumerator3
    Uniform Sprite Size Max=15
    UsePositionOffset=false
    Write Material Random=true
   )
   &CameraOffset(Camera Offset Amount=5)
   /SetVariables(
    Particles.RandFloat:NiagaraFloat=&RandomRangeFloat(Minimum=0, Maximum=1)
    Particles.RandForAttr:NiagaraInt32=&RandomRangeInt(Minimum=0, Maximum=2)
    Particles.RandInt:NiagaraInt32=&RandomRangeInt(Minimum=0, Maximum=239)
   ) {isSetParameters=true, module=SetVariables_#8}
  ]
  particleUpdateScript [
   &ParticleState()
   &self:SetRandomPositionOnChain(
    GroupNum=8
    Particle Attribute Reader=DataInterface() {propertyValues {EmitterBinding {bindingMode=Other, emitterName=Chains_Hands}}}
    ParticleNum=72
    RandFloat:NiagaraFloat=$Particles.RandFloat
    RandInt:NiagaraInt32=$Particles.RandInt
   ) {module=SetRandomPositionOnChain}
   &ScaleSpriteSize(Uniform Curve Sprite Scale=DataInterface() {propertyValues {Curve {keys [
        {interpMode=RCIM_Cubic, tangentMode=RCTM_User, time=0, value=0}
        {interpMode=RCIM_Cubic, tangentMode=RCTM_User, time=0.5, value=1}
        {tangentMode=RCTM_User, time=1, value=0}
       ]}}})
   &SpriteRotationRate(Rotation Rate=&RandomRangeFloat(Minimum=200, Maximum=360))
   &JitterPosition(Jitter Amount=1, Jitter Delay=0.25)
   &ScaleColor(
    Scale RGB=&VectorFromFloat(Value=&FNTimeOfDayScalar(Night Scale=0.5, Morning Scale=0.75, Evening Scale=0.75))
    ScaleA=true
    ScaleRGB=true
   )
  ]
  rendererClasses [&NiagaraSpriteRendererProperties]
  renderers [{rendererClass=&NiagaraSpriteRendererProperties}]
 }
 Lights_Flashing {
  emitterUpdateScript [
   &EmitterState(Life Cycle Mode=::NewEnumerator1, Loop Duration=0.4, UseLoopDelay=false) {module=EmitterState001}
   &SpawnBurst_Instantaneous(Use Spawn Probability=false, UseLoopCountLimit=false)
  ]
  particleSpawnScript [
   &PartitionParticles()
   &InitializeParticle(
    Lifetime=0.4
    Mesh Scale=<2 2 1>
    Mesh Scale Mode=::NewEnumerator4
    Sprite Size Mode=::NewEnumerator3
    Uniform Sprite Size=150
    UsePositionOffset=false
    Write Material Random=true
   )
   &SkeletalMeshLocation(
    Bone / Socket Index=4
    Bone Sampling Mode=::NewEnumerator6
    MaskSpawnGroup=false
    Orientation Sampling=::NewEnumerator1
    Sampled Position Coordinate Space=::Local
    Sampled Position Offset=&MaskVector(Mask Parameter=$Emitter.SpawnGroup, Mask=$Particles.PartitionParticles.Partition, Masked Vector=<0 -111 -33>, Passthrough Vector=<0 -111 33>)
    Skeletal Mesh=$System.Skeletal Mesh
   )
  ]
  particleUpdateScript [
   &ParticleState()
   /SetVariables(
    Particles.FlickerLights:NiagaraFloat=&Waveform(
     "Waveform [1]"=::NewEnumerator12
     "[1] Input Position"=&RandomRangeFloat(Minimum=0, Maximum=1)
     "[1] Frequency"=&FN_FrontEndFloatModifier(Gameplay=20, Front End=10)
     "[1] Amplitude Min/Max"=<0.75 1.5>
    )
    Particles.Jitter:NiagaraFloat=&Waveform(
     "Waveform [1]"=::NewEnumerator12
     "[1] Input Position"=&RandomRangeFloat(Minimum=0, Maximum=1)
     "[1] Frequency"=&FN_FrontEndFloatModifier(Gameplay=50, Front End=10)
     "[1] Amplitude Min/Max"=<0.5 1.5>
    )
   ) {isSetParameters=true}
   &ScaleColor(
    Scale Alpha=&FloatFromCurve(
     FloatCurve=DataInterface() {propertyValues {Curve {keys [{tangentMode=RCTM_User, time=0, value=0}, <at 0.5 1>, {tangentMode=RCTM_User, time=1, value=0}]}}}
     Scale Curve=&Multiply_Float(A=3, B=$Particles.Jitter)
    )
    Scale RGB=&MakeVectorFromLinearColorRGB(Color=&SelectLinearColorFromArray(
      Array Sampling Mode=::NewEnumerator1
      Direct Array Index=$Particles.PartitionParticles.Partition
      Color Selection Array=DataInterface() {propertyValues {ColorData [<rgba 1 0 0 1>, <rgba 0 0.1157 1 1>]}}
      Random Seed=?
     ))
    ScaleA=true
    ScaleRGB=true
   ) {module=ScaleColor001}
   /SetVariables(Particles.LightColor=&MultiplyLinearColor(Linear Color=$Particles.Color, Scale Factor=&MakeVector4_FromFloat(Float=$Particles.FlickerLights))) {isSetParameters=true}
  ]
  rendererClasses [&NiagaraLightRendererProperties, &NiagaraSpriteRendererProperties]
  renderers [
   {rendererClass=&NiagaraLightRendererProperties, rendererIndex=0}
   {rendererClass=&NiagaraSpriteRendererProperties, rendererIndex=1}
  ]
 }
 Pendants_Chains {
  emitterUpdateScript [
   &EmitterState(
    Inactive Response=::NewEnumerator1
    Life Cycle Mode=::NewEnumerator1
    Loop Behavior=::NewEnumerator1
    Loop Duration Mode=::NewEnumerator1
    UseLoopDelay=false
   )
   &SpawnBurst_Instantaneous(Spawn Count=4, Use Spawn Probability=false, UseLoopCountLimit=false)
  ]
  particleSpawnScript [
   &InitializeParticle(
    Lifetime Max=6
    Lifetime Min=3
    Lifetime Mode=::NewEnumerator1
    Mesh Scale Mode=::NewEnumerator3
    Mesh Uniform Scale=0.09
    Sprite Rotation Mode=::NewEnumerator2
    Sprite Size Mode=::NewEnumerator1
    Sprite UV Mode=::NewEnumerator3
    Uniform Sprite Size Max=15
    Uniform Sprite Size Min=10
    UsePositionOffset=false
    Write Material Random=true
   )
   &CameraOffset(Camera Offset Amount=5)
   &InitialMeshOrientation(
    Mesh Axis To Orient=<0 1 0>
    Mesh Orientation Mode=::NewEnumerator2
    Orientation Coordinate Space=::World
    Rotation=<0.5 0 0>
    Use Rotation Vector=true
   )
   &DynamicMaterialParameters(
    Index 0 Param 1=2
    Index 0 Param 2=2
    Index 0 Param 3=2
    Index 0 Param 4=2
    Index 1 Param 1=&SelectFloatFromArray(
     Direct Array Index=&ReturnExecIndex()
     Float Selection Array=DataInterface() {propertyValues {FloatData [0,0,0,0]}}
    )
    Module.Index0Float0Write=true
    Module.Index0Float1Write=true
    Module.Index0Float2Write=true
    Module.Index0Float3Write=true
    Module.Index1Float0Write=true
    Module.Index1Float1Write=false
    Module.Index1Float2Write=false
    Module.Index1Float3Write=false
    Write Parameter Index 1=::NewEnumerator1
   )
  ]
  particleUpdateScript [
   &ParticleState(Kill Particles When Lifetime Has Elapsed=false, Loop Particles Lifetime=true)
   &self:SetPositionOnChain(
    New Particle Attribute Reader=DataInterface() {propertyValues {EmitterBinding {bindingMode=Other, emitterName=CrossHands_Chains}}}
    Particle Index:NiagaraInt32=&SelectIntFromArray(
     Direct Array Index=&ReturnExecIndex()
     Direct Array Mode=::NewEnumerator1
     Int Selection Array=DataInterface() {propertyValues {IntData [36,112,108,116]}}
    )
   ) {module=SetPositionOnChain}
   &SpriteRotationRate(Rotation Rate=&RandomRangeFloat(Minimum=200, Maximum=360))
   &ScaleSpriteSize(
    Uniform Curve Index=&ModuloFloat(
     A=&Add_Float(A=$Particles.NormalizedAge, B=&RandomRangeFloat(Minimum=0, Maximum=1))
     B=1
    )
    Uniform Curve Sprite Scale=DataInterface() {propertyValues {Curve {keys [
        {arriveTangent=10.2355, interpMode=RCIM_Cubic, leaveTangent=10.2355, tangentMode=RCTM_User, time=0, value=0}
        {arriveTangent=0.0075, interpMode=RCIM_Cubic, leaveTangent=0.0075, tangentMode=RCTM_User, time=0.0666, value=1}
        {arriveTangent=-0.0405, interpMode=RCIM_Cubic, leaveTangent=-0.0405, tangentMode=RCTM_User, time=0.15, value=0}
       ]}}}
   )
   &FN_Sprite_Attributes(
    Camera Offset:NiagaraFloat=&Add_Float(A=$Particles.CameraOffset, B=5)
    Create Camera Offset Attribute=true
    Write to Module Namespace=true
   )
   &InterpolateOverTime(
    Rate Of Change=100
    Target Vector Value:Vector3f=&Multiply_Vector(A=$Output.SetPositionOnChain.Direction Vector, B=<1 1 0>)
    Variable Type=::NewEnumerator2
   )
   &UpdateMeshOrientation(
    Facing Direction=&Lerp_Vector(A=$Particles.InterpolateOverTime.InterpolateOverTime_V3_.Moving Average, B=<1 1 1>, Alpha=0.5)
    Facing Interpolation=true
    Facing Interpolation Rate=15
    Orientation Method=::NewEnumerator1
    Override Quat=false
    TransformFacing=false
    TransformUp=false
   )
   &SolveForcesAndVelocity(Clamp Velocity=false, Limit Acceleration=false)
   /SetVariables(Particles.MeshIndex=&SelectIntFromArray(
     Direct Array Index=&ReturnExecIndex()
     Int Selection Array=DataInterface() {propertyValues {IntData [2,0,1,1]}}
    )) {isSetParameters=true}
  ]
  rendererClasses [&NiagaraMeshRendererProperties, &NiagaraSpriteRendererProperties]
  renderers [
   {rendererClass=&NiagaraMeshRendererProperties, rendererIndex=0}
   {rendererClass=&NiagaraSpriteRendererProperties, rendererIndex=1}
  ]
 }
}
summary {
 systemName=NS_Contrail_RoyalAngst
 emitters [
  {emitterName=CrossHands_Chains, rendererClasses [&NiagaraRibbonRendererProperties]}
  {
   emitterName=Pendants_Chains
   rendererClasses [&NiagaraMeshRendererProperties, &NiagaraSpriteRendererProperties]
  }
  {emitterName=Chains_Hands, rendererClasses [&NiagaraRibbonRendererProperties]}
  {emitterName=Chains_Hands_Small, rendererClasses [&NiagaraRibbonRendererProperties]}
  {emitterName=Chains_Feet, rendererClasses [&NiagaraRibbonRendererProperties]}
  {
   emitterName=Lights_Flashing
   rendererClasses [&NiagaraLightRendererProperties, &NiagaraSpriteRendererProperties]
  }
  {emitterName=Glints_Hands, rendererClasses [&NiagaraSpriteRendererProperties]}
  {emitterName=Glints_CrossHands, rendererClasses [&NiagaraSpriteRendererProperties]}
  {emitterName=Glints_Feet, rendererClasses [&NiagaraSpriteRendererProperties]}
 ]
 userVariables [
  User.bIsFrontEnd {type=NiagaraBool}
  User.bIsFrontEndPreview {type=NiagaraBool}
  User.ParaGlideLeanAlpha {type=NiagaraFloat, defaultValue=0, description=""}
  User.PlayerSpeed {type=NiagaraFloat, defaultValue=0, description=""}
  User.RibbonIDOffset {type=NiagaraInt32}
  User.SkeletalMesh {
   type=NiagaraDataInterfaceSkeletalMesh
   defaultValue {
    struct=NiagaraExt_VariableValue_DataInterface
    value {dataInterface=&self:NiagaraDataInterfaceSkeletalMesh_0}
   }
  }
 ]
}
